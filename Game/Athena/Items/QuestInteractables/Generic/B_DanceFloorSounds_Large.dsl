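; Dance floor sound actor (large variant) that changes how filtered its audio sounds as actors enter and leave its overlap area.
; BeginPlay enables the low-pass filter at its minimum frequency and binds to the binding actor's death effects event, which fades the sound out.
; Any overlap plays a timeline sweeping cutoff between LPFMin and LPFMax; end overlap reverses it.
; Play rate is the inverse of FaderTime, so that variable sets sweep duration.
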(event EventBeginPlay
  (|CallParentFunction)
  (Audio|Components|Audio|SetLowPassFilterEnabled (Variables|Sound|GetAudioComponent) true :!class /Script/Engine.AudioComponent)
  (Audio|Components|Audio|SetLowPassFilterFrequency (Variables|Sound|GetAudioComponent) (Variables|Default|GetLPFMin) :!class /Script/Engine.AudioComponent)
  (bind _bindingactor (Variables|Default|GetBindingActor)
    (:then
      (EventDispatchers|AssignOnDeathPlayEffectsDelegate _bindingactor (delegate Custom|OnDeathPlayEffectsDelegate_Event)))
    (:else)))

(event Custom|OnDeathPlayEffectsDelegate_Event
  (params (Damage :type Float%20%28single-precision%29) (DamageTags :type Gameplay%20Tag%20Container%20Structure%20%28by%20ref%29) (Momentum :type Vector) (HitInfo :type Hit%20Result%20Structure%20%28by%20ref%29) (InstigatedBy :type Fort%20Pawn%20Object%20Reference) (DamageCauser :type Actor%20Object%20Reference) (EffectContext :type Gameplay%20Effect%20Context%20Handle%20Structure))
  (Audio|Components|Audio|FadeOut (Variables|Sound|GetAudioComponent) 0.5 0.0 "Linear" :!class /Script/Engine.AudioComponent))

(event EventTick
  (params (DeltaSeconds :type Float%20%28single-precision%29))
  (|CallParentFunction DeltaSeconds) ; This node is disabled and will not be called.\nDrag off pins to build functionality.
)

(event Collision|EventActorBeginOverlap
  (params (OtherActor :type Actor%20Object%20Reference))
  (Components|Timeline|SetPlayRate (Variables|B_DanceFloorSounds_Large|GetLPF%20Fader) (/ 1.0 (Variables|Default|GetFaderTime)))
  (bind (_direction _lpf_normalized) (Timeline|LPF%20Fader Play)
    (:Update
      (Audio|Components|Audio|SetLowPassFilterFrequency (Variables|Sound|GetAudioComponent) (Math|Float|MapRangeUnclamped _lpf_normalized 0.0 1.0 (Variables|Default|GetLPFMin) (Variables|Default|GetLPFMax)) :!class /Script/Engine.AudioComponent))))

(event Collision|EventActorEndOverlap
  (params (OtherActor :type Actor%20Object%20Reference))
  (Components|Timeline|SetPlayRate (Variables|B_DanceFloorSounds_Large|GetLPF%20Fader) (/ 1.0 (Variables|Default|GetFaderTime)))
  (Timeline|LPF%20Fader Reverse))

(fn ConstructionScript
  (|CallParentFunction))
