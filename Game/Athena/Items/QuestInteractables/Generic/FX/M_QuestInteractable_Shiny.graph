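# Masked lit material for shiny quest interactables that bob, sway slightly and glow to stand out in the world.
# A time-driven sine drives both a vertical offset and a small Z-axis rotation, with normals corrected for the rotation.
# Emissive sums a switchable rim glow and a periodic sweep band projected along a camera-space direction, masked by inverted Fresnel.
# An Opacity parameter offsets a noise texture for a dissolve-style mask.
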Material /Game/Athena/Items/QuestInteractables/Generic/FX/M_QuestInteractable_Shiny
graph {
 properties {bTangentSpaceNormal=false, blendMode=BLEND_Masked, pixelDepthOffsetMode=PDOM_Legacy, usedShadingModels=Default Lit}
 nodes [
  n1 = Constant3Vector() {constant=<rgba 0 0 1 0>}
  n2 = Sine(None=Multiply(A=Time(), B=ScalarParameter() {defaultValue=0.4, parameterName=BobRate}))
  n3 = LinearInterpolate(A=Constant() {r=-0.01}, B=Constant() {r=0.01}, Alpha=@n2)
  n4 = DotProduct(
   A=CameraVectorWS()
   B=MaterialFunctionCall(Rotation Axis=@n1, Rotation Angle=@n3) {materialFunction=&FixRotateAboutAxisNormals}.New World Space Vertex Normal
  )
  n5 = Power(Base=OneMinus(None=Clamp(None=@n4))) {constExponent=4}
  n6 = VectorParameter() {parameterName=GlowColour, defaultValue=<rgba 9.6244 10 9.6475 0>}
  n7 = Multiply(A=Subtract(A=WorldPosition().XYZ, B=ObjectPositionWS()), B=ScalarParameter() {defaultValue=0.001, parameterName=WorldTileScale})
  n8 = Normalize(VectorInput=Transform(None=VectorParameter() {parameterName=SweepDirectionCamSpace, defaultValue=<rgba -0.5 0.45 -1 0>}.RGB) {transformSourceType=TRANSFORMSOURCE_Camera})
  n9 = ComponentMask(None=Multiply(A=DotProduct(A=@n7, B=@n8), B=@n8)) {b=true, r=true}
  n10 = Constant() {r=0.28}
  n11 = Frac(None=ComponentMask(None=Panner(Coordinate=@n9, Speed=@n10) {speedX=0.2, speedY=0.2}) {g=true})
  n12 = Add(A=@n11) {constB=0.1}
  n13 = OneMinus(None=Multiply(A=Subtract(A=@n11) {constB=0.9}) {constB=10})
  n14 = MaterialFunctionCall() {materialFunction=&MF_Fresnel_FX}
  n15 = Multiply(A=Multiply(A=Floor(None=@n12), B=Power(Base=@n13) {constExponent=3}), B=Saturate(None=Add(A=@n14.Fresnel Inverted) {constB=0.1}))
  n16 = VectorParameter() {parameterName=SweepColour, defaultValue=<rgba 1.8958 2 1.9078 0>}
  Add(
   A=StaticSwitchParameter(True=Multiply(A=@n5, B=@n6.RGB), False=Constant()) {defaultValue=true, parameterName=UseOuterGlow}
   B=StaticSwitchParameter(True=Multiply(A=@n15, B=@n16.RGB), False=Constant()) {defaultValue=true, parameterName=UseSweep}
  )
  Add(
   A=LinearInterpolate(
    A=VectorParameter() {parameterName=BobHeightMax, defaultValue=<rgba 0 0 10 0>}.RGB
    B=VectorParameter() {parameterName=BobHeightMin, defaultValue=<rgba 0 0 -2 0>}.RGB
    Alpha=@n2
   )
   B=RotateAboutAxis(NormalizedRotationAxis=@n1, RotationAngle=@n3, PivotPoint=ObjectPositionWS(), Position=WorldPosition().XYZ)
  )
  Constant() {r=0.001}
  Constant() {r=1}
  Constant() {r=0.9}
  Constant() {r=0.1}
  MaterialFunctionCall() {materialFunction=&ComponentPivotLocation}
  n17 = Divide(A=ObjectPositionWS(), B=Constant() {r=55})
  n18 = ComponentMask(None=@n17) {r=true}
  n19 = ComponentMask(None=@n17) {g=true}
  Multiply(B=Multiply(
    A=TextureSample(UVs=@n18) {samplerType=SAMPLERTYPE_Masks, texture=/Game/Environments/Sets/Arid/Textures/T_TimeRamp}.R
    B=TextureSample(UVs=@n19) {samplerType=SAMPLERTYPE_Masks, texture=&T_TimeRamp}.R
   ))
  n20 = ScalarParameter() {defaultValue=1, parameterName=Opacity}
  Subtract(A=Add(A=TextureSample() {texture=&T_NoisePack_01}.R), B=Multiply(A=OneMinus(None=@n20)) {constB=2})
  VectorParameter() {parameterName=BaseColor, defaultValue=<rgba 1 1 1 0>}
 ]
}
