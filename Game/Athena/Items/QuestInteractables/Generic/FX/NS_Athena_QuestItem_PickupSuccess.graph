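# Niagara system for the effect played when a quest item pickup succeeds in Athena.
# It layers four emitters named Light, Ground_Ring, Rays and SparkleStationary, suggesting a glow, ground ring, rays and stationary sparkles.
# SystemState is the only listed module, and no renderers, user variables, dynamic inputs or data interfaces are listed.
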NiagaraSystem /Game/Athena/Items/QuestInteractables/Generic/FX/NS_Athena_QuestItem_PickupSuccess
dependencies {usedDataInterfaces [], usedDynamicInputs [], usedModules [&SystemState]}
emitters {
 Ground_Ring {rendererClasses [], renderers []}
 Light {rendererClasses [], renderers []}
 Rays {rendererClasses [], renderers []}
 SparkleStationary {rendererClasses [], renderers []}
}
summary {
 systemName=NS_Athena_QuestItem_PickupSuccess
 emitters [
  {emitterName=Light, rendererClasses []}
  {emitterName=Ground_Ring, rendererClasses []}
  {emitterName=Rays, rendererClasses []}
  {emitterName=SparkleStationary, rendererClasses []}
 ]
 userVariables []
}
