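; Parent Blueprint for Athena quest interactable props, shown and interactable according to the local player's quest state.
; After a workaround delay and playlist load, it hides in blocked playlists; otherwise clients check quest progress and party assist to set local validity.
; Quest updates trigger success feedback and optional hiding, with an optional server completion callback.
; Empty events serve as child hooks; interaction checks validity and cosmetic tags, and a debug flag bypasses quest checks.
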(event Custom|Quest%20Initialized%20and%20Not%20Valid)

(event EventBeginPlay
  (bind _returnvalue (Fortnite|GetFirstLocalFortPlayerController true))
  (bind _questitem (Variables|QuestStuff|GetQuestItem))
  (if (Variables|Default|GetNeverHide)
    (CallFunction|ShowQuestProp))
  (if (Variables|Default|GetForceDebugOn)
    ;[1 0.1602 0.2599] Debug
    (Networking|SwitchHasAuthority
      (:Authority
        _)
      (:Remote
        _))
    (Variables|Default|SetisLocallyValid true)
    (Interaction|BindEventtoOnInteract (EventDispatchers|CreateEvent :!event Custom|DebugModeForceInteract))
    (CallFunction|ShowQuestProp)
    (else
      ;[1 0.5445 0.1188] Workaround - needs to be something better for ship
      (Utilities|FlowControl|Delay 5.0)
      ;[1 1 1] CLIENT AND SERVER - Ensure playlist is loaded
      (bind _asfort_game_state_athena (Utilities|Casting|CastToFortGameStateAthena (Game|GetGameState)))
      (if (Variables|/Script/FortniteGame.FortGameStateAthena|GetbPlaylistDataIsLoaded _asfort_game_state_athena)
        ;[1 1 1] Blocked playlists
        (if (GameplayTags|HasAnyTags (Class|FortGameStateAthena|GetAthenaPlaylistContextTags _asfort_game_state_athena) (Variables|Default|GetBlockedPlaylistTags))
          (if (not (Variables|Default|GetNeverHide))
            (CallFunction|HideQuestProp))
          (else
            ;[1 1 1] Is someone sharing this quest
            (Networking|SwitchHasAuthority
              (:Remote
                (bind (_bisassistingobjective _bisvalidplayercontroller) (Class|AthenaQuestFunctionLibrary|AthenaPartyAssistCheck (Fortnite|GetFirstLocalFortPlayerController true) (Variables|Default|GetQuestBackendObjectiveName)))
                (Variables|Default|SetbIsQuestShared (and _bisassistingobjective _bisvalidplayercontroller))
                ;[1 1 1] Decide if Object should be visible to player or not depending on if they have the quest and objective or not
                (bind (_questisvalid _questcompleted _objectivecompleted) (Default|AthenaCheckQuestandObjectiveCompletion _returnvalue _questitem (Variables|Default|GetQuestBackendObjectiveName)))
                (if _questcompleted
                  ;[1 1 1] if the interactable stays visible after completion but turns off interaction, then we need to turn on interaction when a party member is quest sharing this.
                  (Variables|Default|SetQuestAlreadyFinished true)
                  (if (Variables|Default|GetShouldStayVisibleAfterQuestCompletion)
                    (if (Variables|Default|GetbIsQuestShared)
                      (Variables|Default|SetisLocallyValid true)
                      (else
                        (Variables|Default|SetisLocallyValid false)))
                    (CallFunction|QuestInitializedAndNotValid)
                    (CallFunction|ShowQuestProp)
                    (else
                      ;[1 1 1] If the quest is shared then we don't want to hide it and we want it to be valid
                      (if (not (Variables|Default|GetbIsQuestShared)) :!label _label
                        ;[1 1 1] Hide unless marked not to
                        (Variables|Default|SetisLocallyValid false)
                        (CallFunction|QuestInitializedAndNotValid)
                        (if (Variables|Default|GetShouldStayVisibleAfterQuestCompletion)
                          (CallFunction|ShowQuestProp)
                          (else
                            _)))))
                  (elif (and _questisvalid (not _objectivecompleted))
                    (if (not (Quest|CanProgressQuestinCurrentZone (Quests|GetQuestManager _returnvalue "Invalid") _questitem))
                      (goto _label))
                    (else
                      (goto _label))))
                ;[1 1 1] Quest is valid
                (Variables|Default|SetisLocallyValid true)
                (CallFunction|QuestInitializedandValid)
                (CallFunction|ShowQuestProp)
                ;[1 1 1] Bind Event for When Quest Gets Updated
                (bind _asfort_player_controller_athena (Utilities|Casting|CastToFortPlayerControllerAthena (Fortnite|GetFirstLocalFortPlayerController true)))
                (Quests|BindEventtoOnQuestsUpdated (Quests|GetQuestManager _asfort_player_controller_athena "Invalid") (EventDispatchers|CreateEvent :!event Custom|ParentQuestUpdated))))))
        (else
          (Playlist|BindEventtoOnCurrentPlaylistLoaded _asfort_game_state_athena (delegate Custom|PlaylistLoaded)))))))

(event Custom|QuestInitializedAndValid)

(event Custom|PlaylistLoaded
  (params (PlaylistName :type Name) (PlaylistContextTags :type Gameplay%20Tag%20Container%20Structure%20%28by%20ref%29))
  (bind _returnvalue (Fortnite|GetFirstLocalFortPlayerController true))
  (bind _questitem (Variables|QuestStuff|GetQuestItem))
  ;[1 1 1] Blocked playlists
  (if (GameplayTags|HasAnyTags (Class|FortGameStateAthena|GetAthenaPlaylistContextTags (stale EventBeginPlay _asfort_game_state_athena)) (Variables|Default|GetBlockedPlaylistTags))
    (if (not (Variables|Default|GetNeverHide))
      (CallFunction|HideQuestProp))
    (else
      ;[1 1 1] Is someone sharing this quest
      (Networking|SwitchHasAuthority
        (:Remote
          (bind (_bisassistingobjective _bisvalidplayercontroller) (Class|AthenaQuestFunctionLibrary|AthenaPartyAssistCheck (Fortnite|GetFirstLocalFortPlayerController true) (Variables|Default|GetQuestBackendObjectiveName)))
          (Variables|Default|SetbIsQuestShared (and _bisassistingobjective _bisvalidplayercontroller))
          ;[1 1 1] Decide if Object should be visible to player or not depending on if they have the quest and objective or not
          (bind (_questisvalid _questcompleted _objectivecompleted) (Default|AthenaCheckQuestandObjectiveCompletion _returnvalue _questitem (Variables|Default|GetQuestBackendObjectiveName)))
          (if _questcompleted
            ;[1 1 1] if the interactable stays visible after completion but turns off interaction, then we need to turn on interaction when a party member is quest sharing this.
            (Variables|Default|SetQuestAlreadyFinished true)
            (if (Variables|Default|GetShouldStayVisibleAfterQuestCompletion)
              (if (Variables|Default|GetbIsQuestShared)
                (Variables|Default|SetisLocallyValid true)
                (else
                  (Variables|Default|SetisLocallyValid false)))
              (CallFunction|QuestInitializedAndNotValid)
              (CallFunction|ShowQuestProp)
              (else
                ;[1 1 1] If the quest is shared then we don't want to hide it and we want it to be valid
                (if (not (Variables|Default|GetbIsQuestShared)) :!label _label
                  ;[1 1 1] Hide unless marked not to
                  (Variables|Default|SetisLocallyValid false)
                  (CallFunction|QuestInitializedAndNotValid)
                  (if (Variables|Default|GetShouldStayVisibleAfterQuestCompletion)
                    (CallFunction|ShowQuestProp)
                    (else
                      _)))))
            (elif (and _questisvalid (not _objectivecompleted))
              (if (not (Quest|CanProgressQuestinCurrentZone (Quests|GetQuestManager _returnvalue "Invalid") _questitem))
                (goto _label))
              (else
                (goto _label))))
          ;[1 1 1] Quest is valid
          (Variables|Default|SetisLocallyValid true)
          (CallFunction|QuestInitializedandValid)
          (CallFunction|ShowQuestProp)
          ;[1 1 1] Bind Event for When Quest Gets Updated
          (bind _asfort_player_controller_athena (Utilities|Casting|CastToFortPlayerControllerAthena (Fortnite|GetFirstLocalFortPlayerController true)))
          (Quests|BindEventtoOnQuestsUpdated (Quests|GetQuestManager _asfort_player_controller_athena "Invalid") (EventDispatchers|CreateEvent :!event Custom|ParentQuestUpdated)))))))

(event Custom|ParentQuestUpdated
  ;[1 1 1] Parent Quest Updated
  (Development|PrintString "Hello" false false)
  (Networking|SwitchHasAuthority
    (:Remote
      (Utilities|FlowControl|Sequence
        (:then_0
          ;[1 1 1] Tell Children of Quest Being Updated
          (if (Variables|Default|GetListenForAllQuestUpdates)
            (CallFunction|QuestUpdatedfromAnywhere)))
        (:then_1
          (if (Variables|Default|GetbIsQuestShared)
            _
            (elif (Variables|Default|GetisLocallyValid)
              _))
          ;[1 1 1] Make Interactable Non-Locally Valid
          ; -Check if no longer party assisting this quest if that was true
          ; -Then check if we ourselves have completed it.
          ;[1 1 1] Party assist fix
          (if (Variables|Default|GetbIsQuestShared)
            (bind _bisassistingobjective (Class|AthenaQuestFunctionLibrary|AthenaPartyAssistCheck (Fortnite|GetFirstLocalFortPlayerController true) (Variables|Default|GetQuestBackendObjectiveName)))
            (if _bisassistingobjective
              _
              (else
                _)))
          (bind (_questisvalid _questcompleted _objectivecompleted) (Default|AthenaCheckQuestandObjectiveCompletion (Fortnite|GetFirstLocalFortPlayerController true) (Variables|QuestStuff|GetQuestItem) (Variables|Default|GetQuestBackendObjectiveName)))
          (if (or _questcompleted _objectivecompleted)
            (CallFunction|PlaySuccessFX)
            (Variables|Default|SetisLocallyValid false)
            (Variables|Default|SetbIsQuestShared false)
            ;[1 1 1] UnBind Event for When Quest Gets Updated
            (bind _asfort_player_controller_athena (Utilities|Casting|CastToFortPlayerControllerAthena (Fortnite|GetFirstLocalFortPlayerController true)))
            (Quests|UnbindEventfromOnQuestsUpdated (Quests|GetQuestManager _asfort_player_controller_athena "Invalid") (EventDispatchers|CreateEvent :!event Custom|ParentQuestUpdated))))))))

(event Custom|QuestUpdatedFromAnywhere)

(event Interaction|EventBlueprintOnInteract
  (params (InteractingPawn :type Fort%20Pawn%20Object%20Reference) (InteractionBeingAttempted :type EInteractionBeingAttempted%20Enum :default "FirstInteraction"))
  ;[1 1 1] Interaction Complete - Grab Interacting Player Pawn for use by Children
  (if (Variables|Default|GetForceDebugOn)
    ;[1 0.1352 0.1551] Debug Forcing
    (CallFunction|DebugModeForceInteract 0 InteractingPawn)
    (else
      (bind _output_get (Variables|Default|SetInteractingPawn InteractingPawn))
      (if (Variables|Default|GetbIsQuestShared)
        ;[1 1 1] Our only in point when doing this quest sharing refactor is the interact but it gets called on the server. so we make this event multicasted and call it right here after interact
        (CallFunction|ParentQuestUpdated))
      (if (Variables|Default|GetRunQuestObjectiveCompletionOnServer)
        (bind _asathena_player_controller (Utilities|Casting|CastToAthena_PlayerController (Pawn|GetController _output_get)))
        (EventDispatchers|BindEventtoOnQuestObjectiveStateChanged _asathena_player_controller (EventDispatchers|CreateEvent :!event Custom|Server_QuestObjectiveUpdated))))))

(event Custom|PlaySuccessFX
  ;[1 1 1] Play Succes FX
  (Class|AthenaQuestFunctionLibrary|AthenaPlayQuestRewardFeedback (Variables|Default|GetShouldPlayFXOnPlayer) (Variables|Default|GetSuccessSound) (Variables|Default|GetSuccessNiagara) (Variables|Default|GetSuccessFF) self)
  (CallFunction|ThisQuestObjectiveComplete (Variables|Default|GetInteractingPawn))
  ;[1 1 1] Should Hide after success?
  (if (or (Variables|Default|GetShouldStayVisibleAfterQuestCompletion) (Variables|Default|GetNeverHide))
    _
    (elif (Variables|Default|GetMeshShrinkOnSuccess)
      ;[1 1 1] Should Shrink Before Hiding?
      (bind (_direction _shrinktime) (Timeline|Shrink PlayFromStart)
        (:Update
          (Transformation|SetRelativeScale3D (Variables|Debug|GetStaticMeshComponent) (Math|Vector|Lerp(Vector) "1.000000,1.000000,1.000000" "0.010000,0.010000,0.010000" _shrinktime)))
        (:Finished
          _))
      (elif (Variables|Default|GetLetChildBPHideProp)
        _)))
  (CallFunction|HideQuestProp))

(event Custom|ThisQuestObjectiveComplete
  (params (IneractingPawn :type Fort%20Pawn%20Object%20Reference)))

(event Custom|DebugModeForceInteract
  (params (SelfActor :type Building%20Actor%20Object%20Reference) (InteractingPawn :type Fort%20Pawn%20Object%20Reference))
  (Networking|SwitchHasAuthority
    (:Remote
      (Variables|Default|SetisLocallyValid false)
      (CallFunction|PlaySuccessFX))))

(event Custom|Server_QuestObjectiveUpdated
  (params (Controller :type Fort%20Player%20Controller%20Object%20Reference) (QuestDef :type Fort%20Quest%20Item%20Definition%20Object%20Reference) (BackendName :type Name) (CompletionCount :type Integer) (ObjectiveCompleted :type Boolean) (QuestCompleted :type Boolean))
  ;[1 1 1] Success FX that run on the server
  (if (Utilities|Equal(Object) QuestDef (Variables|QuestStuff|GetQuestItem))
    (if (Utilities|Name|Equal(Name) BackendName (Variables|Default|GetQuestBackendObjectiveName))
      (if (or ObjectiveCompleted QuestCompleted)
        (CallFunction|ServerQuestObjectiveCompleted (Variables|Default|GetInteractingPawn)))
      (EventDispatchers|UnbindEventfromOnQuestObjectiveStateChanged Controller (EventDispatchers|CreateEvent :!event Custom|Server_QuestObjectiveUpdated)))))

(event Custom|Server_QuestObjectiveCompleted
  (params (InteractingPawn :type Fort%20Pawn%20Object%20Reference)))

(fn BlueprintCanInteract
  (params (InteractingPawn :type Fort%20Pawn%20Object%20Reference) (InteractionBeingAttempted :type EInteractionBeingAttempted%20Enum :default "FirstInteraction") (InteractionType :type TInteractionType%20Enum :default "IT_NoInteraction"))
  (bind _islocallyvalid (Variables|Default|GetisLocallyValid))
  (Networking|SwitchHasAuthority
    (:Authority
      (return true))
    (:Remote
      (Utilities|IsValid InteractingPawn
        (:Is%20Valid
          (if (Utilities|GameplayTagContainerNotEmpty (Variables|Default|GetCosmeticCustomTag))
            (bind _returnvalue (GameplayTags|GetOwnedGameplayTags (Variables|Default|GetInteractingPawn) :!class /Script/GameplayTags.GameplayTagAssetInterface))
            (if (GameplayTags|HasAllTags _returnvalue (Variables|Default|GetCosmeticCustomTag) true)
              (LocalVariables|SetLocalVar_InteractQuestValid true)
              (else
                (LocalVariables|SetLocalVar_InteractQuestValid false)))
            (return (and (LocalVariables|GetLocalVar_InteractQuestValid) _islocallyvalid))
            (else
              (return _islocallyvalid))))))))

(fn BlueprintGetInteractionString
  (params (InteractingPawn :type Fort%20Pawn%20Object%20Reference) (InteractionBeingAttempted :type EInteractionBeingAttempted%20Enum :default "FirstInteraction"))
  (return (Variables|Default|GetInteractText)))

(fn HideQuestProp
  (Rendering|SetActorHiddenInGame true)
  (Rendering|Components|Light|SetIntensity (Variables|Default|GetAmbientLight) 0.0)
  (Components|Activation|Deactivate (Variables|Default|GetAmbientSound)))

(fn ShowQuestProp
  (Rendering|SetActorHiddenInGame false)
  (Development|PrintString "ShowQuestProp"))

(fn ConstructionScript
  (if (Variables|Default|GetDisableLightAndSoundComponents)
    (Audio|Components|Audio|SetVolumeMultiplier (Variables|Default|GetAmbientSound) 0.0 :!class /Script/Engine.AudioComponent)
    (Rendering|Components|Light|SetIntensity (Variables|Default|GetAmbientLight) 0.0)))
